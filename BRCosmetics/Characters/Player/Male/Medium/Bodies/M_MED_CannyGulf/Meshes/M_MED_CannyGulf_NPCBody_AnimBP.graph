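# Post-copy animation Blueprint for the M_MED_CannyGulf NPC body mesh, which follows the pose of its attached parent mesh.
# The copied pose runs through the body's Control Rig, then a Constraint node drives the radio joint.
# That joint blends fifteen spine, neck, head, left shoulder and deform bones with uneven, skin-weight-like values.
# The event graph has no logic, and the Simulation Space variable is unused.
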AnimBlueprint /BRCosmetics/Characters/Player/Male/Medium/Bodies/M_MED_CannyGulf/Meshes/M_MED_CannyGulf_NPCBody_AnimBP
target_skeleton=&/M_MED_CannyGulf_NPCBody_Skeleton
graphs [
 AnimGraph {nodes [
   n1 = Control Rig(Source=Copy Pose From Mesh() {bUseAttachedParent=true}) {controlRigAssetReference {blueprintRigClass=&/M_MED_CannyGulf_NPCBody_CtrlRig.M_MED_CannyGulf_NPCBody_CtrlRig_C}}
   n2 = "Constraint - dyn_radio_jnt (PPPPPPPPPPPPPPP)"!(
    ComponentPose=Local To Component(LocalPose=@n1)
    Alpha=1
    ConstraintWeights_0=0.38
    ConstraintWeights_1:float=0.2794
    ConstraintWeights_2:float=0.0948
    ConstraintWeights_3:float=0.0882
    ConstraintWeights_4:float=0.0536
    ConstraintWeights_5:float=0.0221
    ConstraintWeights_6:float=0.0183
    ConstraintWeights_7:float=0.0161
    ConstraintWeights_8:float=0.0154
    ConstraintWeights_9:float=0.0141
    ConstraintWeights_10:float=0.0061
    ConstraintWeights_11:float=0.0044
    ConstraintWeights_12:float=0.003
    ConstraintWeights_13:float=0.0026
    ConstraintWeights_14:float=0.002
   ) -> (Pose) {
    boneToModify {boneName=dyn_radio_jnt}
    constraintSetup [
     {targetBone {boneName=spine_05}}
     {targetBone {boneName=clavicle_l}}
     {targetBone {boneName=spine_03}}
     {targetBone {boneName=upperarm_twist_01_l}}
     {targetBone {boneName=spine_04}}
     {targetBone {boneName=neck_02}}
     {targetBone {boneName=deform_upperarm_up_l}}
     {targetBone {boneName=deform_pec_l}}
     {targetBone {boneName=deform_upperarm_fr_l}}
     {targetBone {boneName=neck_01}}
     {targetBone {boneName=upperarm_twist_02_l}}
     {targetBone {boneName=upperarm_l}}
     {targetBone {boneName=deform_upperarm_bk_l}}
     {targetBone {boneName=spine_02}}
     {targetBone {boneName=head}}
    ]
    constraintWeights [1,1,1,1,1,1,1,1,1,1,1,1,1,1,1]
   }
   Output Pose(Result=Component To Local(ComponentPose=@n2)) {layerGroup=None}
  ]}
 EventGraph {nodes [Event BlueprintUpdateAnimation()]}
]
variables [Simulation Space {type=AnimPhysSimSpaceType, default=RootRelative}]
